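# Animation Blueprint for the IronBlaze default female medium body cosmetic, following the attached parent mesh and adding secondary deformation and motion.
# The copied pose runs through a deform Control Rig, then additive bone offsets shift left belt pouches when crouching and right pouches, with interpolation, when crouch-walking.
# A RigidBody node then simulates physics in pelvis space.
# The update event derives both alphas, plus a right-leg bend value the AnimGraph never reads.
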AnimBlueprint /BRCosmetics/Characters/Player/Female/Medium/Bodies/F_MED_IronBlaze_Default/F_MED_IronBlaze_Default_AnimBP
target_skeleton=&/F_MED_IronBlaze_Default_Skeleton
graphs [
 AnimGraph {nodes [
   n1 = Control Rig(Source=Copy Pose From Mesh() {bUseAttachedParent=true}) {
    alphaBoolBlend {blendOption=Linear}
    controlRigAssetReference {blueprintRigClass=&/F_MED_DeformRig_IronBlaze_Default_ControlRig.F_MED_DeformRig_IronBlaze_Default_ControlRig_C}
   }
   n2 = Get Crouching() -> (Crouching:double)
   n3 = "Transform (Modify) Bone - Bone: master_L_front_belt_pouch_jnt"(ComponentPose=Local To Component(LocalPose=@n1), Alpha=@n2, Translation="1.5,-1.6,0", Rotation="0,0,-40", Scale="1,1,1") {
    rotationMode=BMM_Additive
    rotationSpace=BCS_BoneSpace
    translationMode=BMM_Additive
    translationSpace=BCS_BoneSpace
    alphaBoolBlend {blendOption=Linear}
    boneToModify {boneName=master_L_front_belt_pouch_jnt}
    rotation {roll=-40}
    translation=<1.5 -1.6>
   }
   n4 = "Transform (Modify) Bone - Bone: master_L_back_belt_pouch_jnt"(ComponentPose=@n3, Alpha=@n2, Translation="0,-1,0", Rotation="0,0,-15", Scale="1,1,1") {
    rotationMode=BMM_Additive
    rotationSpace=BCS_BoneSpace
    translationMode=BMM_Additive
    translationSpace=BCS_BoneSpace
    alphaBoolBlend {blendOption=Linear}
    boneToModify {boneName=master_L_back_belt_pouch_jnt}
    rotation {roll=-15}
    translation {y=-1}
   }
   n5 = Get Crouch Walking() -> (Crouch Walking:double)
   n6 = "Transform (Modify) Bone - Bone: master_R_front_belt_pouch_jnt"(ComponentPose=@n4, Alpha=@n5, Translation="-2,-3.5,0", Rotation="0,0,20", Scale="1,1,1") {
    rotationMode=BMM_Additive
    rotationSpace=BCS_BoneSpace
    translationMode=BMM_Additive
    translationSpace=BCS_BoneSpace
    alphaBoolBlend {blendOption=Linear}
    alphaScaleBiasClamp {bInterpResult=true, interpSpeedDecreasing=5, interpSpeedIncreasing=2}
    boneToModify {boneName=master_R_front_belt_pouch_jnt}
    rotation {roll=20}
    translation=<-2 -3.5>
   }
   n7 = "Transform (Modify) Bone - Bone: master_R_back_belt_pouch_jnt"(ComponentPose=@n6, Alpha=@n5, Translation="0.5,-1.5,0", Rotation="0,0,15", Scale="1,1,1") {
    rotationMode=BMM_Additive
    rotationSpace=BCS_BoneSpace
    translationMode=BMM_Additive
    translationSpace=BCS_BoneSpace
    alphaBoolBlend {blendOption=Linear}
    alphaScaleBiasClamp {bInterpResult=true, interpSpeedDecreasing=5, interpSpeedIncreasing=2}
    boneToModify {boneName=master_R_back_belt_pouch_jnt}
    rotation {roll=15}
    translation=<0.5 -1.5>
   }
   n8 = RigidBody(ComponentPose=@n7) {
    bClampLinearTranslationLimitToRefPose=true
    bOverrideWorldGravity=true
    overridePhysicsAsset=&/F_MED_IronBlaze_Default_Physics
    simulationSpace=BaseBoneSpace
    alphaBoolBlend {blendOption=Linear}
    baseBoneRef {boneName=pelvis}
    componentAppliedLinearAccClamp=<3000 3000 3000>
    componentLinearAccScale=<0.3 0.2 0.2>
    componentLinearVelScale=<0.3 0.2 0.2>
    overrideWorldGravity {z=-980}
    simSpaceSettings {maxAngularAcceleration=1000, maxAngularVelocity=3, maxLinearAcceleration=250, maxLinearVelocity=500, velocityScaleZ=0, worldAlpha=1}
   }
   Output Pose(Result=Component To Local(ComponentPose=@n8)) {layerGroup=None}
  ]}
 EventGraph {nodes [
   Event BlueprintUpdateAnimation()
   > Set Crouching(Crouching:double="To Float (Boolean)"(InBool=Get bIsCrouching())) -> (Output_Get:double)
   n1 = BreakTransform(InTransform=GetDeltaTransformFromRefPose(self=GetOwnerSkeletalMeshComponent(), BoneName=calf_r, BaseName=thigh_r))
   > Set R Leg Bent(R Leg Bent:double=MapRangeClamped(Value=BreakRotator(InRot=@n1.Rotation).Yaw, InRangeA=90, InRangeB=120)) -> (Output_Get:double)
   n2 = Get bIsCrouching()
   n3 = Get bIsMoving2D()
   > Set Crouch Walking(Crouch Walking:double="To Float (Boolean)"(InBool=AND Boolean(A=@n2, B=@n3))) -> (Output_Get:double)
  ]}
]
variables [
 Crouch Walking {type=number}
 Crouching {type=number}
 R Leg Bent {type=number}
]
